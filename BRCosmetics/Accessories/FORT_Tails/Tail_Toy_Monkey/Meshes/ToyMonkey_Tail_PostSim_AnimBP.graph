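# Post-process AnimBP for the Toy Monkey tail cosmetic, posing and simulating the tail.
# An idle/crouch state machine using Neon Cat clips blends over the input pose from the tail base, then bone offsets feed world-space spring AnimDynamics on nine joints.
# Events toggle a material "Hide Element 01" for cape or long backbling and set skydiving and velocity alphas.
# The Trail controller branch is disconnected, and neither alpha reaches the output pose.
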AnimBlueprint /BRCosmetics/Accessories/FORT_Tails/Tail_Toy_Monkey/Meshes/ToyMonkey_Tail_PostSim_AnimBP
target_skeleton=/Game/Accessories/FORT_Tails/Common/Fortnite_Base_Tail_Skeleton
graphs [
 AnimGraph {nodes [
   n1 = Trail controller - Bone: C_Tail_A_9_Jnt!(
    ComponentPose=Local To Component()
    Alpha=Get trail_alpha() -> (trail_alpha:double)
    RelaxationSpeedScale:float=Get RelaxSpeed() -> (RelaxSpeed:float)
   ) -> (Pose) {
    bLimitRotation=true
    chainLength=10
    maxDeltaTime=0.0333
    alphaScaleBiasClamp {scale=0.55}
    rotationLimits [
     {limitMax=<180 180 180>}
     {limitMax=<180 180 180>}
     {limitMax=<180 180 180>}
     {limitMax=<180 180 180>}
     {limitMax=<180 180 180>}
     {limitMax=<180 180 180>}
     {limitMax=<180 180 180>}
     {limitMax=<180 180 180>}
     {limitMax=<180 180 180>}
     {limitMax=<180 180 180>}
    ]
    rotationOffsets [<0 0 0>, <0 0 0>, <0 0 0>, <0 0 0>, <0 0 0>, <0 0 0>, <0 0 0>, <0 0 0>, <0 0 0>, <0 0 0>]
    trailBone {boneName=C_Tail_A_9_Jnt}
    trailRelaxationSpeed {editorCurveData {keys [<at 0 10 RCIM_Linear>,<at 0.4177 8.6564 RCIM_Linear>,<at 1 5 RCIM_Linear>]}}
   }
   n2 = "Transform (Modify) Bone - Bone: C_Root_Main_Root_Jnt"(ComponentPose=@n1, Alpha=1, Translation="0,1.5,1", Rotation="0,0,0", Scale="1,1,1") {
    scaleMode=BMM_Replace
    scaleSpace=BCS_BoneSpace
    translationMode=BMM_Additive
    translationSpace=BCS_BoneSpace
    alphaBoolBlend {blendOption=Linear}
    boneToModify {boneName=C_Root_Main_Root_Jnt}
    translation {y=1.5, z=1}
   }
   Component To Local(ComponentPose=@n2)
   n3 = "Transform (Modify) Bone - Bone: C_Tail_A_Base_Jnt"(
    ComponentPose=Local To Component(LocalPose=Use cached pose 'Anim+BaseRotation'())
    Alpha=1
    Translation="0,0,0"
    Rotation="0,0,0"
    Scale="1,1,1"
   ) {
    rotationMode=BMM_Additive
    rotationSpace=BCS_ParentBoneSpace
    alphaBoolBlend {blendOption=Linear}
    alphaScaleBiasClamp {interpSpeedDecreasing=5}
    boneToModify {boneName=C_Tail_A_Base_Jnt}
   }
   n4 = "Transform (Modify) Bone - Bone: C_Tail_A_1_Jnt"(ComponentPose=@n3, Alpha=1, Translation="0,0,0", Rotation="10,0,0", Scale="1,1,1") {
    rotationMode=BMM_Additive
    rotationSpace=BCS_BoneSpace
    alphaBoolBlend {blendOption=Linear}
    alphaScaleBiasClamp {interpSpeedDecreasing=5}
    boneToModify {boneName=C_Tail_A_1_Jnt}
    rotation {pitch=10}
   }
   n5 = "Transform (Modify) Bone - Bone: C_Tail_A_2_Jnt"(ComponentPose=@n4, Alpha=1, Translation="0,0,0", Rotation="10,0,0", Scale="1,1,1") {
    rotationMode=BMM_Additive
    rotationSpace=BCS_BoneSpace
    alphaBoolBlend {blendOption=Linear}
    alphaScaleBiasClamp {interpSpeedDecreasing=5}
    boneToModify {boneName=C_Tail_A_2_Jnt}
    rotation {pitch=10}
   }
   n6 = "Transform (Modify) Bone - Bone: C_Tail_A_2_Jnt"(ComponentPose=@n5, Alpha=1, Translation="0,0,0", Rotation="10,0,0", Scale="1,1,1") {
    rotationMode=BMM_Additive
    rotationSpace=BCS_BoneSpace
    alphaBoolBlend {blendOption=Linear}
    alphaScaleBiasClamp {interpSpeedDecreasing=5}
    boneToModify {boneName=C_Tail_A_2_Jnt}
    rotation {pitch=10}
   }
   n7 = "Transform (Modify) Bone - Bone: C_Tail_A_3_Jnt"(ComponentPose=@n6, Alpha=1, Translation="0,0,0", Rotation="5,0,0", Scale="1,1,1") {
    rotationMode=BMM_Additive
    rotationSpace=BCS_BoneSpace
    alphaBoolBlend {blendOption=Linear}
    alphaScaleBiasClamp {interpSpeedDecreasing=5}
    boneToModify {boneName=C_Tail_A_3_Jnt}
    rotation {pitch=5}
   }
   n8 = Anim Dynamics - Bone: C_Tail_A_1_Jnt!(ComponentPose=@n7, Alpha=1) -> (Pose) {
    angularDampingOverride=0.999
    angularSpringConstant=700
    bAngularSpring=true
    bOverrideAngularDamping=true
    bUseGravityOverride=true
    gravityScale=0.35
    simulationSpace=World
    alphaBoolBlend {blendOption=Linear}
    alphaScaleBiasClamp {bInterpResult=true, scale=0.5}
    boundBone {boneName=C_Tail_A_1_Jnt}
    componentAppliedLinearAccClamp=<1000 1000 1000>
    componentLinearAccScale=<1 1 1>
    componentLinearVelScale=<1 1 1>
    gravityOverride {z=-1300}
    physicsBodyDefinitions [{
      boundBone {boneName=C_Tail_A_1_Jnt}
      boxExtents=<12 12 12>
      constraintSetup {angularLimitsMax=<0 3 25>, angularLimitsMin=<0 -50 -25>, angularTarget=<1 0 0>}
      localJointOffset=<0.5 0 0>
     }]
   }
   n9 = Anim Dynamics - Bone: C_Tail_A_2_Jnt!(ComponentPose=@n8, Alpha=1) -> (Pose) {
    angularDampingOverride=0.999
    angularSpringConstant=700
    bAngularSpring=true
    bOverrideAngularDamping=true
    bUseGravityOverride=true
    gravityScale=0.35
    simulationSpace=World
    alphaBoolBlend {blendOption=Linear}
    alphaScaleBiasClamp {bInterpResult=true, scale=0.5}
    boundBone {boneName=C_Tail_A_2_Jnt}
    componentAppliedLinearAccClamp=<1000 1000 1000>
    componentLinearAccScale=<1 1 1>
    componentLinearVelScale=<1 1 1>
    gravityOverride {z=-1300}
    physicsBodyDefinitions [{
      boundBone {boneName=C_Tail_A_2_Jnt}
      boxExtents=<12 12 12>
      constraintSetup {angularLimitsMax=<0 3 25>, angularLimitsMin=<0 -50 -25>, angularTarget=<1 0 0>}
      localJointOffset=<0.5 0 0>
     }]
   }
   n10 = Anim Dynamics - Bone: C_Tail_A_3_Jnt!(ComponentPose=@n9, Alpha=1) -> (Pose) {
    angularDampingOverride=0.999
    angularSpringConstant=700
    bAngularSpring=true
    bOverrideAngularDamping=true
    bUseGravityOverride=true
    gravityScale=0.35
    simulationSpace=World
    alphaBoolBlend {blendOption=Linear}
    alphaScaleBiasClamp {bInterpResult=true, scale=0.5}
    boundBone {boneName=C_Tail_A_3_Jnt}
    componentAppliedLinearAccClamp=<1000 1000 1000>
    componentLinearAccScale=<1 1 1>
    componentLinearVelScale=<1 1 1>
    gravityOverride {z=-1300}
    physicsBodyDefinitions [{
      boundBone {boneName=C_Tail_A_3_Jnt}
      boxExtents=<12 12 12>
      constraintSetup {angularLimitsMax=<0 3 25>, angularLimitsMin=<0 -50 -25>, angularTarget=<1 0 0>}
      localJointOffset=<0.5 0 0>
     }]
   }
   n11 = Anim Dynamics - Bone: C_Tail_A_4_Jnt!(ComponentPose=@n10, Alpha=1) -> (Pose) {
    angularDampingOverride=0.999
    angularSpringConstant=700
    bAngularSpring=true
    bOverrideAngularDamping=true
    bUseGravityOverride=true
    gravityScale=0.35
    simulationSpace=World
    alphaBoolBlend {blendOption=Linear}
    alphaScaleBiasClamp {bInterpResult=true, scale=0.5}
    boundBone {boneName=C_Tail_A_4_Jnt}
    componentAppliedLinearAccClamp=<1000 1000 1000>
    componentLinearAccScale=<1 1 1>
    componentLinearVelScale=<1 1 1>
    gravityOverride {z=-1300}
    physicsBodyDefinitions [{
      boundBone {boneName=C_Tail_A_4_Jnt}
      boxExtents=<12 12 12>
      constraintSetup {angularLimitsMax=<0 3 25>, angularLimitsMin=<0 -45 -25>, angularTarget=<1 0 0>}
      localJointOffset=<0.5 0 0>
     }]
   }
   n12 = Anim Dynamics - Bone: C_Tail_A_5_Jnt!(ComponentPose=@n11, Alpha=1) -> (Pose) {
    angularDampingOverride=0.999
    angularSpringConstant=700
    bAngularSpring=true
    bOverrideAngularDamping=true
    bUseGravityOverride=true
    gravityScale=0.35
    simulationSpace=World
    alphaBoolBlend {blendOption=Linear}
    alphaScaleBiasClamp {bInterpResult=true, scale=0.5}
    boundBone {boneName=C_Tail_A_5_Jnt}
    componentAppliedLinearAccClamp=<1000 1000 1000>
    componentLinearAccScale=<1 1 1>
    componentLinearVelScale=<1 1 1>
    gravityOverride {z=-1300}
    physicsBodyDefinitions [{
      boundBone {boneName=C_Tail_A_5_Jnt}
      boxExtents=<13 13 13>
      constraintSetup {angularLimitsMax=<0 3 25>, angularLimitsMin=<0 -40 -25>, angularTarget=<1 0 0>}
      localJointOffset=<0.5 0 0>
     }]
   }
   n13 = Anim Dynamics - Bone: C_Tail_A_6_Jnt!(ComponentPose=@n12, Alpha=1) -> (Pose) {
    angularDampingOverride=0.999
    angularSpringConstant=700
    bAngularSpring=true
    bOverrideAngularDamping=true
    bUseGravityOverride=true
    gravityScale=0.35
    simulationSpace=World
    alphaBoolBlend {blendOption=Linear}
    alphaScaleBiasClamp {bInterpResult=true, scale=0.5}
    boundBone {boneName=C_Tail_A_6_Jnt}
    componentAppliedLinearAccClamp=<1000 1000 1000>
    componentLinearAccScale=<1 1 1>
    componentLinearVelScale=<1 1 1>
    gravityOverride {z=-1300}
    physicsBodyDefinitions [{
      boundBone {boneName=C_Tail_A_6_Jnt}
      boxExtents=<14 14 14>
      constraintSetup {angularLimitsMax=<0 3 25>, angularLimitsMin=<0 -35 -25>, angularTarget=<1 0 0>}
      localJointOffset=<0.5 0 0>
     }]
   }
   n14 = Anim Dynamics - Bone: C_Tail_A_7_Jnt!(ComponentPose=@n13, Alpha=1) -> (Pose) {
    angularDampingOverride=0.999
    angularSpringConstant=700
    bAngularSpring=true
    bOverrideAngularDamping=true
    bUseGravityOverride=true
    gravityScale=0.35
    simulationSpace=World
    alphaBoolBlend {blendOption=Linear}
    alphaScaleBiasClamp {bInterpResult=true, scale=0.5}
    boundBone {boneName=C_Tail_A_7_Jnt}
    componentAppliedLinearAccClamp=<1000 1000 1000>
    componentLinearAccScale=<1 1 1>
    componentLinearVelScale=<1 1 1>
    gravityOverride {z=-1300}
    physicsBodyDefinitions [{
      boundBone {boneName=C_Tail_A_7_Jnt}
      boxExtents=<15 15 15>
      constraintSetup {angularLimitsMax=<0 3 25>, angularLimitsMin=<0 -30 -25>, angularTarget=<1 0 0>}
      localJointOffset=<0.5 0 0>
     }]
   }
   n15 = Anim Dynamics - Bone: C_Tail_A_8_Jnt!(ComponentPose=@n14, Alpha=1) -> (Pose) {
    angularDampingOverride=0.999
    angularSpringConstant=700
    bAngularSpring=true
    bOverrideAngularDamping=true
    bUseGravityOverride=true
    gravityScale=0.35
    simulationSpace=World
    alphaBoolBlend {blendOption=Linear}
    alphaScaleBiasClamp {bInterpResult=true, scale=0.5}
    boundBone {boneName=C_Tail_A_8_Jnt}
    componentAppliedLinearAccClamp=<1000 1000 1000>
    componentLinearAccScale=<1 1 1>
    componentLinearVelScale=<1 1 1>
    gravityOverride {z=-1300}
    physicsBodyDefinitions [{
      boundBone {boneName=C_Tail_A_8_Jnt}
      boxExtents=<16 16 16>
      constraintSetup {angularLimitsMax=<0 3 25>, angularLimitsMin=<0 -25 -25>, angularTarget=<1 0 0>}
      localJointOffset=<0.5 0 0>
     }]
   }
   n16 = Anim Dynamics - Bone: C_Tail_A_9_Jnt!(ComponentPose=@n15, Alpha=1) -> (Pose) {
    angularDampingOverride=0.999
    angularSpringConstant=700
    bAngularSpring=true
    bOverrideAngularDamping=true
    bUseGravityOverride=true
    gravityScale=0.35
    simulationSpace=World
    alphaBoolBlend {blendOption=Linear}
    alphaScaleBiasClamp {bInterpResult=true, scale=0.5}
    boundBone {boneName=C_Tail_A_9_Jnt}
    componentAppliedLinearAccClamp=<1000 1000 1000>
    componentLinearAccScale=<1 1 1>
    componentLinearVelScale=<1 1 1>
    gravityOverride {z=-1300}
    physicsBodyDefinitions [{
      boundBone {boneName=C_Tail_A_9_Jnt}
      boxExtents=<17 17 17>
      constraintSetup {angularLimitsMax=<0 3 25>, angularLimitsMin=<0 -20 -25>, angularTarget=<1 0 0>}
      localJointOffset=<0.5 0 0>
     }]
   }
   Output Pose(Result=Component To Local(ComponentPose=@n16)) {layerGroup=None}
   AnimGraphNode_StateMachine_0 = toy_monkey_tail_state() {is=title AnimGraphNode_StateMachine}
   AnimGraphNode_SubInput_0 = Input Pose()
   n17 = "Transform (Modify) Bone - Bone: C_Root_Main_Root_Jnt"(
    ComponentPose=Local To Component(LocalPose=Layered blend per bone(BasePose=@AnimGraphNode_SubInput_0, BlendPoses_0=@AnimGraphNode_StateMachine_0) {layerSetup [{branchFilters [{boneName=C_Tail_A_Base_Jnt}]}]})
    Alpha=1
    Translation="0,0,0"
    Rotation="0,0,-15"
    Scale="1,1,1"
   ) {
    rotationMode=BMM_Additive
    rotationSpace=BCS_BoneSpace
    alphaBoolBlend {blendOption=Linear}
    boneToModify {boneName=C_Root_Main_Root_Jnt}
    rotation {roll=-15}
   }
   Save cached pose 'Anim+BaseRotation'(Pose=Component To Local(ComponentPose=@n17)) {cacheName=Anim+BaseRotation}
  ]}
 toy_monkey_tail_state {
  entry=Idle
  path=AnimGraph.AnimGraphNode_StateMachine_0.toy_monkey_tail_state
  ids=id
  nodes [
   AnimStateNode_0 = Idle()
   AnimStateNode_1 = Crouch() {is=title AnimStateNode}
   AnimStateTransitionNode_0 = AnimStateTransitionNode()
   AnimStateTransitionNode_2 = AnimStateTransitionNode()
  ]
  links [
   {
    dst=Idle
    rule=AnimGraph.AnimGraphNode_StateMachine_0.toy_monkey_tail_state.AnimStateTransitionNode_2.Transition
    src=Crouch
    via=AnimStateTransitionNode_2
   }
   {
    dst=Crouch
    rule=AnimGraph.AnimGraphNode_StateMachine_0.toy_monkey_tail_state.AnimStateTransitionNode_0.Transition
    src=Idle
    via=AnimStateTransitionNode_0
   }
  ]
 }
 Idle {
  path=AnimGraph.AnimGraphNode_StateMachine_0.toy_monkey_tail_state.AnimStateNode_0.Idle
  nodes [Output Animation Pose(Result=Sequence Player 'Tail_Idle'() {sequence=/Game/Accessories/FORT_Tails/Tail_Neon_Cat/Anims/Tail_Idle})]
 }
 Crouch {
  path=AnimGraph.AnimGraphNode_StateMachine_0.toy_monkey_tail_state.AnimStateNode_1.Crouch
  nodes [
   AnimGraphNode_StateMachine_0 = Crouch() {is=title AnimGraphNode_StateMachine}
   Output Animation Pose(Result=@AnimGraphNode_StateMachine_0)
  ]
 }
 Crouch {
  entry=Blank
  path=AnimGraph.AnimGraphNode_StateMachine_0.toy_monkey_tail_state.AnimStateNode_1.Crouch.AnimGraphNode_StateMachine_0.Crouch
  ids=id
  nodes [
   AnimStateNode_0 = Blank() {is=title AnimStateNode}
   AnimStateNode_1 = Sliding() {is=title AnimStateNode}
   AnimStateNode_2 = CrouchSprinting() {is=title AnimStateNode}
   AnimStateNode_27 = CrouchWalk() {is=title AnimStateNode}
   AnimStateNode_29 = Crouch() {is=title AnimStateNode}
   AnimStateTransitionNode_0 = AnimStateTransitionNode()
   AnimStateTransitionNode_1 = AnimStateTransitionNode()
   AnimStateTransitionNode_11 = AnimStateTransitionNode()
   AnimStateTransitionNode_12 = AnimStateTransitionNode()
   AnimStateTransitionNode_2 = AnimStateTransitionNode()
   AnimStateTransitionNode_3 = AnimStateTransitionNode()
   AnimStateTransitionNode_4 = AnimStateTransitionNode()
   AnimStateTransitionNode_47 = AnimStateTransitionNode()
   AnimStateTransitionNode_55 = AnimStateTransitionNode()
   AnimStateTransitionNode_7 = AnimStateTransitionNode()
  ]
  links [
   {
    dst=Crouch
    rule=AnimGraph.AnimGraphNode_StateMachine_0.toy_monkey_tail_state.AnimStateNode_1.Crouch.AnimGraphNode_StateMachine_0.Crouch.AnimStateTransitionNode_11.Transition
    src=Blank
    via=AnimStateTransitionNode_11
   }
   {
    dst=CrouchWalk
    rule=AnimGraph.AnimGraphNode_StateMachine_0.toy_monkey_tail_state.AnimStateNode_1.Crouch.AnimGraphNode_StateMachine_0.Crouch.AnimStateTransitionNode_12.Transition
    src=Blank
    via=AnimStateTransitionNode_12
   }
   {
    dst=CrouchWalk
    rule=AnimGraph.AnimGraphNode_StateMachine_0.toy_monkey_tail_state.AnimStateNode_1.Crouch.AnimGraphNode_StateMachine_0.Crouch.AnimStateTransitionNode_7.Transition
    src=Crouch
    via=AnimStateTransitionNode_7
   }
   {
    dst=Sliding
    rule=AnimGraph.AnimGraphNode_StateMachine_0.toy_monkey_tail_state.AnimStateNode_1.Crouch.AnimGraphNode_StateMachine_0.Crouch.AnimStateTransitionNode_0.Transition
    src=Crouch
    via=AnimStateTransitionNode_0
   }
   {
    dst=CrouchWalk
    rule=AnimGraph.AnimGraphNode_StateMachine_0.toy_monkey_tail_state.AnimStateNode_1.Crouch.AnimGraphNode_StateMachine_0.Crouch.AnimStateTransitionNode_55.AnimationTransitionGraph_1
    src=CrouchSprinting
    via=AnimStateTransitionNode_55
   }
   {
    dst=Crouch
    rule=AnimGraph.AnimGraphNode_StateMachine_0.toy_monkey_tail_state.AnimStateNode_1.Crouch.AnimGraphNode_StateMachine_0.Crouch.AnimStateTransitionNode_4.Transition
    src=CrouchWalk
    via=AnimStateTransitionNode_4
   }
   {
    dst=CrouchSprinting
    rule=AnimGraph.AnimGraphNode_StateMachine_0.toy_monkey_tail_state.AnimStateNode_1.Crouch.AnimGraphNode_StateMachine_0.Crouch.AnimStateTransitionNode_47.AnimationTransitionGraph_1
    src=CrouchWalk
    via=AnimStateTransitionNode_47
   }
   {
    dst=Sliding
    rule=AnimGraph.AnimGraphNode_StateMachine_0.toy_monkey_tail_state.AnimStateNode_1.Crouch.AnimGraphNode_StateMachine_0.Crouch.AnimStateTransitionNode_3.Transition
    src=CrouchWalk
    via=AnimStateTransitionNode_3
   }
   {
    dst=Crouch
    rule=AnimGraph.AnimGraphNode_StateMachine_0.toy_monkey_tail_state.AnimStateNode_1.Crouch.AnimGraphNode_StateMachine_0.Crouch.AnimStateTransitionNode_1.Transition
    src=Sliding
    via=AnimStateTransitionNode_1
   }
   {
    dst=CrouchWalk
    rule=AnimGraph.AnimGraphNode_StateMachine_0.toy_monkey_tail_state.AnimStateNode_1.Crouch.AnimGraphNode_StateMachine_0.Crouch.AnimStateTransitionNode_2.Transition
    src=Sliding
    via=AnimStateTransitionNode_2
   }
  ]
 }
 Blank {
  path=AnimGraph.AnimGraphNode_StateMachine_0.toy_monkey_tail_state.AnimStateNode_1.Crouch.AnimGraphNode_StateMachine_0.Crouch.AnimStateNode_0.Blank
  nodes [Output Animation Pose()]
 }
 Sliding {
  path=AnimGraph.AnimGraphNode_StateMachine_0.toy_monkey_tail_state.AnimStateNode_1.Crouch.AnimGraphNode_StateMachine_0.Crouch.AnimStateNode_1.Sliding
  nodes [Output Animation Pose()]
 }
 CrouchSprinting {
  path=AnimGraph.AnimGraphNode_StateMachine_0.toy_monkey_tail_state.AnimStateNode_1.Crouch.AnimGraphNode_StateMachine_0.Crouch.AnimStateNode_2.CrouchSprinting
  nodes [Output Animation Pose(Result=Sequence Player 'Tail_Crouch_Sprint'() {sequence=/Game/Accessories/FORT_Tails/Tail_Neon_Cat/Anims/Tail_Crouch_Sprint})]
 }
 CrouchWalk {
  path=AnimGraph.AnimGraphNode_StateMachine_0.toy_monkey_tail_state.AnimStateNode_1.Crouch.AnimGraphNode_StateMachine_0.Crouch.AnimStateNode_27.CrouchWalk
  nodes [Output Animation Pose(Result=Sequence Player 'Tail_Crouch_Walk'() {sequence=/Game/Accessories/FORT_Tails/Tail_Neon_Cat/Anims/Tail_Crouch_Walk})]
 }
 Crouch {
  path=AnimGraph.AnimGraphNode_StateMachine_0.toy_monkey_tail_state.AnimStateNode_1.Crouch.AnimGraphNode_StateMachine_0.Crouch.AnimStateNode_29.Crouch
  nodes [
   n1 = "Transform (Modify) Bone - Bone: C_Tail_A_3_Jnt"(
    ComponentPose=Local To Component(LocalPose=Sequence Player 'Tail_Crouch_Idle'() {sequence=/Game/Accessories/FORT_Tails/Tail_Neon_Cat/Anims/Tail_Crouch_Idle})
    Alpha=1
    Translation="0,0,0"
    Rotation="20,0,0"
    Scale="1,1,1"
   ) {
    rotationMode=BMM_Additive
    rotationSpace=BCS_BoneSpace
    alphaBoolBlend {blendOption=Linear}
    boneToModify {boneName=C_Tail_A_3_Jnt}
    rotation {pitch=20}
   }
   n2 = "Transform (Modify) Bone - Bone: C_Tail_A_5_Jnt"(ComponentPose=@n1, Alpha=1, Translation="0,0,0", Rotation="20,0,0", Scale="1,1,1") {
    rotationMode=BMM_Additive
    rotationSpace=BCS_BoneSpace
    alphaBoolBlend {blendOption=Linear}
    boneToModify {boneName=C_Tail_A_5_Jnt}
    rotation {pitch=20}
   }
   n3 = "Transform (Modify) Bone - Bone: C_Tail_A_7_Jnt"(ComponentPose=@n2, Alpha=1, Translation="0,0,0", Rotation="20,0,0", Scale="1,1,1") {
    rotationMode=BMM_Additive
    rotationSpace=BCS_BoneSpace
    alphaBoolBlend {blendOption=Linear}
    boneToModify {boneName=C_Tail_A_7_Jnt}
    rotation {pitch=20}
   }
   Output Animation Pose(Result=Component To Local(ComponentPose=@n3))
  ]
 }
 Transition {
  path=AnimGraph.AnimGraphNode_StateMachine_0.toy_monkey_tail_state.AnimStateNode_1.Crouch.AnimGraphNode_StateMachine_0.Crouch.AnimStateTransitionNode_0.Transition
  nodes [Result(bCanEnterTransition=Get bIsSliding())]
 }
 Transition {
  path=AnimGraph.AnimGraphNode_StateMachine_0.toy_monkey_tail_state.AnimStateNode_1.Crouch.AnimGraphNode_StateMachine_0.Crouch.AnimStateTransitionNode_1.Transition
  nodes [Result(bCanEnterTransition=NOT Boolean(A=Get bIsSliding()))]
 }
 Transition {
  path=AnimGraph.AnimGraphNode_StateMachine_0.toy_monkey_tail_state.AnimStateNode_1.Crouch.AnimGraphNode_StateMachine_0.Crouch.AnimStateTransitionNode_11.Transition
  nodes [Result(bCanEnterTransition=NOT Boolean(A=Get bIsMoving2D()))]
 }
 Transition {
  path=AnimGraph.AnimGraphNode_StateMachine_0.toy_monkey_tail_state.AnimStateNode_1.Crouch.AnimGraphNode_StateMachine_0.Crouch.AnimStateTransitionNode_12.Transition
  nodes [Result(bCanEnterTransition=Get bIsMoving2D())]
 }
 Transition {
  path=AnimGraph.AnimGraphNode_StateMachine_0.toy_monkey_tail_state.AnimStateNode_1.Crouch.AnimGraphNode_StateMachine_0.Crouch.AnimStateTransitionNode_2.Transition
  nodes [Result(bCanEnterTransition=NOT Boolean(A=Get bIsSliding()))]
 }
 Transition {
  path=AnimGraph.AnimGraphNode_StateMachine_0.toy_monkey_tail_state.AnimStateNode_1.Crouch.AnimGraphNode_StateMachine_0.Crouch.AnimStateTransitionNode_3.Transition
  nodes [Result(bCanEnterTransition=Get bIsSliding())]
 }
 Transition {
  path=AnimGraph.AnimGraphNode_StateMachine_0.toy_monkey_tail_state.AnimStateNode_1.Crouch.AnimGraphNode_StateMachine_0.Crouch.AnimStateTransitionNode_4.Transition
  nodes [Result(bCanEnterTransition=NOT Boolean(A=Get bIsMoving2D()))]
 }
 AnimationTransitionGraph_1 {
  path=AnimGraph.AnimGraphNode_StateMachine_0.toy_monkey_tail_state.AnimStateNode_1.Crouch.AnimGraphNode_StateMachine_0.Crouch.AnimStateTransitionNode_47.AnimationTransitionGraph_1
  nodes [Result(bCanEnterTransition=Get bIsCrouchSprinting())]
 }
 AnimationTransitionGraph_1 {
  path=AnimGraph.AnimGraphNode_StateMachine_0.toy_monkey_tail_state.AnimStateNode_1.Crouch.AnimGraphNode_StateMachine_0.Crouch.AnimStateTransitionNode_55.AnimationTransitionGraph_1
  nodes [Result(bCanEnterTransition=NOT Boolean(A=Get bIsCrouchSprinting()))]
 }
 Transition {
  path=AnimGraph.AnimGraphNode_StateMachine_0.toy_monkey_tail_state.AnimStateNode_1.Crouch.AnimGraphNode_StateMachine_0.Crouch.AnimStateTransitionNode_7.Transition
  nodes [Result(bCanEnterTransition=Get bIsMoving2D())]
 }
 Transition {
  path=AnimGraph.AnimGraphNode_StateMachine_0.toy_monkey_tail_state.AnimStateTransitionNode_0.Transition
  nodes [
   Result(bCanEnterTransition=Get bIsCrouching())
   OR Boolean(A=Get bTransition_Idle_Crouching() -> (bTransition_Idle_Crouching:bool), B=false)
  ]
 }
 Transition {
  path=AnimGraph.AnimGraphNode_StateMachine_0.toy_monkey_tail_state.AnimStateTransitionNode_2.Transition
  nodes [Result(bCanEnterTransition=NOT Boolean(A=Get bIsCrouching()))]
 }
 EventGraph {nodes [
   Event BlueprintUpdateAnimation()
   > n1 = Branch(Condition=AND Boolean(A=Get bIsSkyDiving(), B=Get bIsInFrontEnd()))
   > Set highVelocityTurnOffAnimDyn(highVelocityTurnOffAnimDyn:double=0) -> (Output_Get:double)
   n2 = BreakVector(InVec=Get PawnVelocity())
   n3 = "float < float"(A="Absolute (Float)"(A=@n2), B=1300)
   n4 = "float < float"(A="Absolute (Float)"(A=@n2.Y), B=1300)
   n5 = "float < float"(A="Absolute (Float)"(A=@n2.Z), B=1300)
   Set highVelocityTurnOffAnimDyn(execute=@n1.else, highVelocityTurnOffAnimDyn="To Float (Boolean)"(InBool=AND Boolean!(A=@n3, B=@n4, C:bool=@n5) -> (ReturnValue)))
   Event OnCharacterCustomizationLoaded()
   n6 = CheckForBackbling()
   > n7 = Branch(Condition=OR Boolean(A=@n6.bIsBackblingACape, B=@n6.bIsBackblingLong))
   > n8 = Get Dynamic Material Instance(SkeletalMeshComponent=GetPartSkeletalMeshComponent())
   > SetScalarParameterValue(self=@n8, ParameterName=Hide Element 01, Value=1)
   SetScalarParameterValue(execute=@n7.else, self=@n8, ParameterName=Hide Element 01, Value=0)
   Event OnEndSkydiving()
   > Set trail_alpha(trail_alpha:double=1) -> (Output_Get:double)
   Event OnBeginSkydiving()
   > Set trail_alpha(trail_alpha=0)
  ]}
]
variables [
 highVelocityTurnOffAnimDyn {type=number}
 trail_alpha {type=number, default=1}
]
